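; BP_Commerce_Bed_PlaceSetting_A is an Asteria Commerce bed place-setting prop Blueprint that includes a bed bench mesh among its components.
; Its construction script sets a fixed muted pink-beige color on the SM_Commerce_BedBench_B secondary mesh.
; When the actor is destroyed, it destroys its Collision_1 component.
; BeginPlay, overlap and Tick events are empty.
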(event EventBeginPlay)

(event Collision|EventActorBeginOverlap
  (params (OtherActor :type Actor%20Object%20Reference)))

(event EventTick
  (params (DeltaSeconds :type Float%20%28single-precision%29)))

(event EventDestroyed
  (Components|DestroyComponent (Variables|Default|GetCollision_1)))

(fn ConstructionScript
  (Default|SetColoronSecondaryMesh "(R=0.571125,G=0.407240,B=0.381326,A=1.000000)" (Variables|Default|GetSM_Commerce_BedBench_B)))
